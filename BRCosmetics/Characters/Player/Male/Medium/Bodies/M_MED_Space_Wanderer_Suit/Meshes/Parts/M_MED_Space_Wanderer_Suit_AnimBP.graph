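# Animation Blueprint for the Space Wanderer Suit body part of the medium male player cosmetic.
# It copies the pose from the attached parent mesh and runs it through the suit's face Control Rig to produce the output pose.
# An Anim Dynamics setup for a wire bone and a HelmRotation cached-pose blend on neck and face bones sit outside the output chain.
# The event graph holds only an empty update event.
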AnimBlueprint /BRCosmetics/Characters/Player/Male/Medium/Bodies/M_MED_Space_Wanderer_Suit/Meshes/Parts/M_MED_Space_Wanderer_Suit_AnimBP
target_skeleton=&/M_MED_Space_Wanderer_Suit_Skeleton
graphs [
 AnimGraph {nodes [
   n1 = Anim Dynamics - Bone: dyn_wire(ComponentPose=Local To Component(), Alpha=0, ExternalForce="0,0,0") {
    angularDampingOverride=0.9
    angularSpringConstant=3500
    bAngularSpring=true
    bGravityOverrideInSimSpace=true
    bOverrideAngularDamping=true
    bUseGravityOverride=true
    gravityScale=0.35
    simulationSpace=RootRelative
    alphaBoolBlend {blendOption=Linear}
    boundBone {boneName=dyn_wire}
    componentAppliedLinearAccClamp=<3000 3000 3000>
    componentLinearAccScale=<0.2 0.2 0.2>
    componentLinearVelScale=<0.2 0.2 0.2>
    gravityOverride {z=-500}
    physicsBodyDefinitions [{
      sphereCollisionRadius=0
      boundBone {boneName=dyn_wire}
      boxExtents=<10 10 10>
      constraintSetup {angularLimitsMax=<0 10 0>, angularLimitsMin=<0 -10 -10>, angularTarget=<1 0 0>}
      localJointOffset=<0 0 -2>
     }]
   }
   Component To Local(ComponentPose=@n1)
   Layered blend per bone!(
    BasePose=Use cached pose 'HelmRotation'()
    BlendPoses_0
    BlendPoses_1:Pose Link Structure=Use cached pose 'HelmRotation'()
    BlendWeights_0=0
    BlendWeights_1:float=1
   ) -> (Pose) {
    blendPoses [{linkId=-1, sourceLinkId=-1}, {linkId=-1, sourceLinkId=-1}]
    blendWeights [1,1]
    layerSetup [{branchFilters [{blendDepth=1, boneName=neck_01}]}, {branchFilters [{blendDepth=1, boneName=faceAttach}]}]
   }
   n2 = Control Rig(Source=Copy Pose From Mesh() {bUseAttachedParent=true}) {
    alphaBoolBlend {blendOption=Linear}
    controlRigAssetReference {blueprintRigClass=&/M_MED_Space_Wanderer_Suit_Face_CTRrig.M_MED_Space_Wanderer_Suit_Face_CTRrig_C}
   }
   Output Pose(Result=@n2) {layerGroup=None}
   Save cached pose 'HelmRotation'() {cacheName=HelmRotation}
  ]}
 EventGraph {nodes [Event BlueprintUpdateAnimation()]}
]
variables []
